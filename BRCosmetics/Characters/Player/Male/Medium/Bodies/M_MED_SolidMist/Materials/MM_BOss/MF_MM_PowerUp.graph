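# Material function that adds a power-up emissive glow to incoming material attributes for the SolidMist character body.
# Two emissive colors and mask-blended strengths set the glow, which is added to the existing emissive.
# Its coverage comes from inverted, powered curvature and AO terms maxed together, then shaped by further rerouted inputs.
# Named reroutes publish input mask channels, panning distorted soft noise, fresnel masks and a switchable FX texture mask.
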MaterialFunction /BRCosmetics/Characters/Player/Male/Medium/Bodies/M_MED_SolidMist/Materials/MM_BOss/MF_MM_PowerUp
graph {nodes [
  BlendMaterialAttributes(A=NamedRerouteUsage(), B=NamedRerouteUsage(), Alpha=NamedRerouteUsage())
  Constant() {r=1}
  Constant() {r=1}
  n1 = GetMaterialAttributes(None=NamedRerouteUsage()) {attributeGetTypes [#1]}
  n2 = VectorParameter() {parameterName=EmissiveColor_02, defaultValue=<rgba 0.6615 0.1333 0 1>}
  n3 = VectorParameter() {parameterName=EmissiveColor, defaultValue=<rgba 1 0.3705 0 1>}
  n4 = NamedRerouteUsage()
  n5 = ScalarParameter() {defaultValue=0.5, parameterName=MM_BG_Emissive_Str}
  n6 = ScalarParameter() {defaultValue=10, parameterName=MM_Emissive_Str}
  n7 = NamedRerouteUsage()
  n8 = Multiply(
   A=Multiply(A=NamedRerouteUsage(), B=LinearInterpolate(A=@n2.RGB, B=@n3.RGB, Alpha=@n4))
   B=LinearInterpolate(
    A=LinearInterpolate(A=@n5, B=@n6, Alpha=@n7) {constA=5}
    B=ScalarParameter() {defaultValue=20, parameterName=MM_Mask_Emissive_Str}
    Alpha=NamedRerouteUsage()
   )
  )
  n9 = ScalarParameter() {defaultValue=1.3, parameterName=EmissivePow}
  n10 = NamedRerouteUsage()
  n11 = NamedRerouteUsage()
  n12 = ScalarParameter() {defaultValue=1.5, parameterName=CurvatureStr}
  n13 = ScalarParameter() {defaultValue=1, parameterName=CurvatureStr_masks}
  n14 = NamedRerouteUsage()
  n15 = Power(
   Base=Multiply(A=OneMinus(None=@n11), B=LinearInterpolate(A=@n12, B=@n13, Alpha=@n14))
   Exp=LinearInterpolate(
    A=ScalarParameter() {defaultValue=6, parameterName=CuravturePower}
    B=ScalarParameter() {defaultValue=6, parameterName=CuravturePower_mask}
    Alpha=NamedRerouteUsage()
   )
  )
  n16 = NamedRerouteUsage()
  n17 = ScalarParameter() {defaultValue=1, parameterName=AO_Str}
  n18 = ScalarParameter() {defaultValue=1.3, parameterName=AO_Str_Mask}
  n19 = NamedRerouteUsage()
  n20 = Power(
   Base=Multiply(A=OneMinus(None=@n16), B=LinearInterpolate(A=@n17, B=@n18, Alpha=@n19))
   Exp=LinearInterpolate(
    A=ScalarParameter() {defaultValue=12, parameterName=AO_Power}
    B=ScalarParameter() {defaultValue=3, parameterName=AO_Power_mask}
    Alpha=NamedRerouteUsage()
   )
  )
  n21 = Max(A=@n10, B=Max(A=Saturate(None=@n15), B=Saturate(None=@n20)))
  n22 = Max(A=Subtract(A=@n21, B=NamedRerouteUsage()), B=Multiply(A=@n21, B=NamedRerouteUsage()))
  n23 = Add(A=@n1.EmissiveColor, B=Multiply(A=Power(Base=@n8, Exp=@n9) {constExponent=1.25}, B=Saturate(None=@n22)))
  FunctionOutput(None=NamedRerouteDeclaration(None=SetMaterialAttributes!(MaterialAttributes=@n1.MaterialAttributes, EmissiveColor=@n23) {attributeSetTypes [#1]}) {name=Mod_Material, nodeColor=<rgba 0 1 0.5547 1>})
  MaterialFunctionCall!(UV Coord, LenghtScale, RadialScale) {materialFunction=/Game/Global/MaterialFunctions/TechArt/MF_PolarCoordinates}
  Max(A=@n10, B=@n22)
  n24 = SetMaterialAttributes!(MaterialAttributes)
  NamedRerouteDeclaration(None=GetMaterialAttributes(None=FunctionInput(Preview=@n24) {inputName=Material, inputType=FunctionInput_MaterialAttributes}).MaterialAttributes) {name=InputMaterial, nodeColor=<rgba 1 0 0.1641 1>}
  n25 = Constant3Vector()
  NamedRerouteDeclaration(None=ComponentMask(None=FunctionInput(Preview=@n25) {bUsePreviewValueAsDefault=true, inputName=Mask, sortPriority=1}) {r=true}) {name=FX_Mask, nodeColor=<rgba 0 0.4844 1 1>}
  n26 = FunctionInput(Preview=@n25) {bUsePreviewValueAsDefault=true, inputName=M_Texture, sortPriority=2}
  NamedRerouteDeclaration(None=ComponentMask(None=@n26) {g=true}) {name=M_Mask_01, nodeColor=<rgba 0.1484 1 0 1>}
  NamedRerouteDeclaration(None=ComponentMask(None=@n26) {r=true}) {name=M_Mask_02, nodeColor=<rgba 0 0.1562 1 1>}
  n27 = TextureCoordinate()
  n28 = Multiply(A=@n27, B=ComponentMask(None=VectorParameter() {parameterName=SoftNoise_Tiling_01, defaultValue=<rgba 6 6 0 1>}.RGB) {g=true, r=true})
  n29 = Multiply(
   A=Time()
   B=ComponentMask(None=VectorParameter() {parameterName=SoftNoise_PanSpeedAndDirection_01, defaultValue=<rgba 0 0.2 0 1>}.RGB) {g=true, r=true}
  )
  n30 = ComponentMask(None=VectorParameter() {parameterName=SoftNoiseDistortion_Tiling_01, defaultValue=<rgba 12 12 0 1>}.RGB) {g=true, r=true}
  n31 = Time()
  n32 = ComponentMask(None=VectorParameter() {parameterName=SoftNoiseDistortion_PanSpeedAndDirection_01, defaultValue=<rgba 0 0.2 0 1>}.RGB) {g=true, r=true}
  n33 = TextureSample(UVs=Add(A=Multiply(A=@n27, B=@n30), B=Multiply(A=@n31, B=@n32))) {texture=/Game/Effects/Fort_Effects/Emotes/UltraLight/Texture/T_SoftNoise_BC7}
  n34 = Multiply(A=Subtract(A=Multiply(A=@n33.RGB) {constB=2}), B=ScalarParameter() {defaultValue=0.1, parameterName=NoiseDistortion_Str})
  n35 = TextureSample(UVs=Add(A=Add(A=@n28, B=@n29), B=ComponentMask(None=@n34) {g=true, r=true})) {texture=&T_SoftNoise_BC7}
  NamedRerouteDeclaration(None=Multiply(A=@n35.R, B=ScalarParameter() {defaultValue=0.5, parameterName=Noise_Str})) {name=Grey_Noise, nodeColor=<rgba 1 0.2109 0 1>}
  NamedRerouteDeclaration(None=Multiply(A=@n35.G, B=ScalarParameter() {defaultValue=3, parameterName=noise_Str_02}) {constB=6}) {name=Grey_Noise_2, nodeColor=<rgba 1 0.2109 0 1>}
  n36 = CameraVectorWS()
  n37 = DotProduct(A=@n36, B=PixelNormalWS())
  n38 = ScalarParameter() {defaultValue=4, parameterName=Fresnelpow}
  NamedRerouteDeclaration(None=Saturate(None=Power(Base=@n37, Exp=@n38))) {name=FresnelMask, nodeColor=<rgba 0 1 0.1094 1>}
  n39 = DotProduct(A=@n36, B=VertexNormalWS())
  NamedRerouteDeclaration(None=Saturate(None=Power(Base=@n39))) {name=FresnelMask2, nodeColor=<rgba 0.1016 0 1 1>}
  n40 = TextureSample() {samplerType=SAMPLERTYPE_Masks, texture=/BRCosmetics/Characters/Player/Male/Medium/Bodies/M_MED_SolidMist/Textures/T_SolidMist_Body_FX2}
  NamedRerouteDeclaration(None=StaticSwitchParameter(True=@n40.R, False=Constant()) {defaultValue=true, parameterName=UseMask2}) {name=FX_Mask_02, nodeColor=<rgba 0.7578 0 1 1>}
 ]}
